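# Material function for the FearlessFlightMenace character body cosmetic that renders paint strokes as unlit, thickened spline geometry.
# SplineThicken drives World Position Offset, with one width parameter feeding both base and tip.
# Vertex color red blends emissive from black to red in a newly set attribute bundle.
# A static switch, on by default, selects between named reroutes, apparently this stroke result or the incoming material attributes.
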MaterialFunction /BRCosmetics/Characters/Player/Male/Medium/Bodies/M_MED_FearlessFlightMenace/Materials/MFC_FearlessFlightMenace_PaintStrokes
graph {nodes [
  FunctionOutput(None=StaticSwitchParameter(True=NamedRerouteUsage(), False=NamedRerouteUsage()) {defaultValue=true, group=01_FearlessFlightMenace, parameterName=UseFearlessFlightMenacePaintStrokes, sortPriority=0})
  NamedRerouteDeclaration(None=FunctionInput() {inputName=MA, inputType=FunctionInput_MaterialAttributes}) {name=MA, nodeColor=<rgba 0.5937 1 0 1>}
  n1 = ScalarParameter() {defaultValue=0.1, group=01_FearlessFlightMenace, parameterName=PaintStroke_Width, sliderMax=1}
  n2 = SetMaterialAttributes!(
   MaterialAttributes
   EmissiveColor=LinearInterpolate(A=Constant3Vector(), B=Constant3Vector() {constant=<rgba 1 0 0 0>}, Alpha=VertexColor().R) {desc=Emissive Color}
   WorldPositionOffset=MaterialFunctionCall(WidthBase=@n1, WidthTip=@n1, Expand U or V UV Channel=StaticBool()) {materialFunction=&SplineThicken}.WorldPositionOffset
   ShadingModel=ShadingModel() {shadingModel=MSM_Unlit}
  ) {
   desc=screen projected rim and hatching
   attributeSetTypes [#1, #2, #3]
  }
  NamedRerouteDeclaration(None=@n2) {name=MA_OUT, nodeColor=<rgba 0 1 0.4141 1>}
  TextureCoordinate()
 ]}
