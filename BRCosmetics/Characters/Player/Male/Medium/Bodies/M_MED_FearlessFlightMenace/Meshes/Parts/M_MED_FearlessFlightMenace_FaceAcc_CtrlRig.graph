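# Control Rig for the FearlessFlightMenace face accessory mesh that procedurally drives trap and chest helper bones from body motion.
# Weighted constraints position traps from upper arms and head, rotate them from clavicles and upper spine, and position chest bones from upper-arm twists and clavicles.
# Spherical pose readers measure each clavicle against spine_05, storing weights that apply corrective trap offsets.
# Right-side trap rotation uses Spine_05_Orient; elbow variables and trapNeck bones are unused.
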ControlRigBlueprint /BRCosmetics/Characters/Player/Male/Medium/Bodies/M_MED_FearlessFlightMenace/Meshes/Parts/M_MED_FearlessFlightMenace_FaceAcc_CtrlRig
graphs [
 RigVMFunctionLibrary {links [], nodes []}
 RigVMModel {nodes [
   RigUnit_BeginExecution = &RigUnit_BeginExecution()
   /Script/ControlRig.RigUnit_PositionConstraintLocalSpaceOffset(
    Child=(Type=Bone,Name="trap_l")
    Filter=(bX=True,bY=True,bZ=True)
    Parents=((Item=(Type=Bone,Name="upperarm_l"),Weight=0.7),(Item=(Type=Bone,Name="head"),Weight=0.3))
    Weight=1
    bMaintainOffset=True
    ExecutePin=@RigUnit_BeginExecution.ExecutePin
   ) {is=struct RigVMUnitNode}
   > &RigUnit_PositionConstraintLocalSpaceOffset(
    Child=(Type=Bone,Name="trap_r")
    Filter=(bX=True,bY=True,bZ=True)
    Parents=((Item=(Type=Bone,Name="upperarm_r"),Weight=0.7),(Item=(Type=Bone,Name="head"),Weight=0.3))
    Weight=1
    bMaintainOffset=True
   )
   > &RigUnit_RotationConstraintLocalSpaceOffset(
    AdvancedSettings=(InterpolationType=Shortest,RotationOrderForFilter=XZY)
    Child=(Type=Bone,Name="trap_l")
    Filter=(bX=True,bY=True,bZ=True)
    Parents=((Item=(Type=Bone,Name="clavicle_l"),Weight=0.5),(Item=(Type=Bone,Name="spine_05"),Weight=0.5))
    Weight=1
   )
   > &RigUnit_RotationConstraintLocalSpaceOffset(
    AdvancedSettings=(InterpolationType=Shortest,RotationOrderForFilter=XZY)
    Child=(Type=Bone,Name="trap_r")
    Filter=(bX=True,bY=True,bZ=True)
    Parents=((Item=(Type=Bone,Name="clavicle_r"),Weight=0.5),(Item=(Type=Bone,Name="Spine_05_Orient"),Weight=0.5))
    Weight=1
   )
   > &RigUnit_PositionConstraintLocalSpaceOffset(
    Child=(Type=Bone,Name="chest_l")
    Filter=(bX=True,bY=True,bZ=True)
    Parents=((Item=(Type=Bone,Name="upperarm_twist_02_l"),Weight=0.5),(Item=(Type=Bone,Name="clavicle_l"),Weight=0.5))
    Weight=1
    bMaintainOffset=True
   )
   > &RigUnit_PositionConstraintLocalSpaceOffset(
    Child=(Type=Bone,Name="chest_r")
    Filter=(bX=True,bY=True,bZ=True)
    Parents=((Item=(Type=Bone,Name="upperarm_twist_02_r"),Weight=0.5),(Item=(Type=Bone,Name="clavicle_r"),Weight=0.5))
    Weight=1
    bMaintainOffset=True
   )
   > n1 = &RigUnit_SphericalPoseReader(
    ActiveRegionScaleFactors=(PositiveWidth=0.1,NegativeWidth=0.1,PositiveHeight=0.1,NegativeHeight=0.1)
    ActiveRegionSize=0.1
    Debug=(bDrawDebug=false,bDraw2D=false,DebugScale=25,DebugSegments=20,DebugThickness=0.25)
    DriverAxis=(X=1,Y=0,Z=0)
    DriverItem=(Type=Bone,Name="clavicle_l")
    FalloffRegionScaleFactors=(PositiveWidth=1,NegativeWidth=1,PositiveHeight=1,NegativeHeight=1)
    FalloffSize=0.45
    OptionalParentItem=(Type=Bone,Name="spine_05")
    RotationOffset=(X=0,Y=0,Z=0)
   )
   > RigVMVariableNode(Value=@n1.OutputParam) {variable=clavicleUp_l}
   > n2 = &RigUnit_SphericalPoseReader(
    ActiveRegionScaleFactors=(PositiveWidth=0.1,NegativeWidth=0.1,PositiveHeight=0.1,NegativeHeight=0.1)
    ActiveRegionSize=0.1
    Debug=(bDrawDebug=false,bDraw2D=false,DebugScale=25,DebugSegments=20,DebugThickness=0.25)
    DriverAxis=(X=-1,Y=0,Z=0)
    DriverItem=(Type=Bone,Name="clavicle_r")
    FalloffRegionScaleFactors=(PositiveWidth=1,NegativeWidth=0,PositiveHeight=1,NegativeHeight=1)
    FalloffSize=0.45
    OptionalParentItem=(Type=Bone,Name="spine_05")
    RotationOffset=(X=0,Y=180,Z=0)
   )
   > RigVMVariableNode(Value=@n2.OutputParam) {variable=clavicleUp_r}
   > &RigUnit_ModifyTransforms(
    ItemToModify=((Item=(Type=Bone,Name="trap_l"),Transform=(Rotation=(X=0,Y=0,Z=0,W=1),Translation=(X=5,Y=-5,Z=0),Scale3D=(X=1,Y=1,Z=1))))
    Weight=RigVMVariableNode() {variable=clavicleUp_l}.Value
   )
   > &RigUnit_ModifyTransforms(
    ItemToModify=((Item=(Type=Bone,Name="trap_r"),Transform=(Rotation=(X=0,Y=0,Z=0,W=1),Translation=(X=-5,Y=5,Z=0),Scale3D=(X=1,Y=1,Z=1))))
    Weight=RigVMVariableNode() {variable=clavicleUp_r}.Value
   )
  ]}
]
hierarchy [
 root {loc=<0 0 0>, rot=<0 0 0 1>}
 pelvis {parent=root, loc=<0 -0.7071 93.1191>, rot=<0.0127 0.707 -0.0127 -0.707>}
 spine_01 {parent=pelvis, loc=<6.7494 -0.6388 0>, rot=<0 0 0.034 -0.9994>}
 spine_02 {parent=spine_01, loc=<7.2518 -0.035 0>, rot=<0 0 0.0238 -0.9997>}
 spine_03 {parent=spine_02, loc=<6.0104 0 0>, rot=<0 0 0.0149 -0.9999>}
 spine_04 {parent=spine_03, loc=<8.1024 0 0>, rot=<0 0 -0.0132 -0.9999>}
 spine_05 {parent=spine_04, loc=<10.0146 0 0>, rot=<0 0 -0.0015 -1>}
 neck_01 {parent=spine_05, loc=<11.3138 0.1529 0>, rot=<0 0 -0.2003 -0.9797>}
 neck_02 {parent=neck_01, loc=<4.5627 0 0>, rot=<0 0 0.0504 -0.9987>}
 head {parent=neck_02, loc=<5.3914 0 0>, rot=<0 0 0.115 -0.9934>}
 clavicle_l {parent=spine_05, loc=<4.3575 -0.1466 -2.3223>, rot=<0.0465 -0.7289 -0.009 -0.6829>}
 upperarm_l {parent=clavicle_l, loc=<15.5111 0 0>, rot=<0.0226 -0.3234 -0.0088 -0.9459>}
 lowerarm_l {parent=upperarm_l, loc=<26.6514 0 0>, rot=<0.0391 -0.0099 -0.2453 -0.9686>}
 lowerarm_twist_02_l {parent=lowerarm_l, loc=<9.6199 0 0>, rot=<0 0 0 -1>}
 lowerarm_twist_01_l {parent=lowerarm_l, loc=<17.8549 0 0>, rot=<0 0 0 -1>}
 upperarm_twist_01_l {parent=upperarm_l, loc=<0.0041 0 0>, rot=<0 0 0 1>}
 upperarm_twist_02_l {parent=upperarm_l, loc=<13.7941 0 0>, rot=<0 0 0 1>}
 clavicle_r {parent=spine_05, loc=<4.3575 -0.1466 2.3223>, rot=<0.7289 0.0465 -0.6829 0.009>}
 upperarm_r {parent=clavicle_r, loc=<-15.5111 0 0>, rot=<0.0226 -0.3234 -0.0088 -0.9459>}
 lowerarm_r {parent=upperarm_r, loc=<-26.6514 0 0>, rot=<0.0391 -0.0099 -0.2453 -0.9686>}
 lowerarm_twist_02_r {parent=lowerarm_r, loc=<-9.6199 0 0>, rot=<0 0 0 -1>}
 lowerarm_twist_01_r {parent=lowerarm_r, loc=<-17.8549 0 0>, rot=<0 0 0 -1>}
 upperarm_twist_01_r {parent=upperarm_r, loc=<-0.0041 0 0>, rot=<0 0 0 -1>}
 upperarm_twist_02_r {parent=upperarm_r, loc=<-13.7941 0 0>, rot=<0 0 0 -1>}
 trap_l {parent=spine_05, loc=<8.1905 -0.3323 -6.767>, rot=<0.0465 -0.7289 -0.009 -0.6829>}
 trap_r {parent=spine_05, loc=<8.1905 -0.3323 6.767>, rot=<0.7289 0.0465 -0.6829 0.009>}
 Spine_05_Orient {parent=spine_05, loc=<0 0 0>, rot=<0.7289 0.0465 -0.6829 0.009>}
 chest_l {parent=spine_05, loc=<-1.62 3.4508 -2.1669>, rot=<0.0284 -0.7065 -0.0284 -0.7065>}
 chest_r {parent=spine_05, loc=<-1.6197 3.4508 2.1669>, rot=<0.7065 0.0284 -0.7065 0.0284>}
 trapNeck_l {parent=spine_05, loc=<13.064 0.06 -5.854>, rot=<0.0465 -0.7289 -0.009 -0.6829>}
 trapNeck_r {parent=spine_05, loc=<13.064 0.06 5.854>, rot=<0.7289 0.0465 -0.6829 0.009>}
]
variables [
 clavicleUp_l {type=float, default=0}
 clavicleUp_r {type=float, default=0}
 elbow_l {type=float, default=0}
 elbow_r {type=float, default=0}
]
