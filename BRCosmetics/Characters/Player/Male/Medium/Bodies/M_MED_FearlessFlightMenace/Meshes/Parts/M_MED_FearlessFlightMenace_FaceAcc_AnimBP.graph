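# Animation Blueprint for the FearlessFlightMenace character's face accessory part, following the body and correcting it per player state.
# Copied parent and owner-mesh poses, layered at the upper arms, pass through a base Control Rig plus rifle and pistol rigs toggled by weapon animation type.
# Additive bone tweaks then rescale trap and chest bones when hang gliding and raise traps when skydive-floating.
# Spring controllers driven by backbling and jump state are disconnected from the output.
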AnimBlueprint /BRCosmetics/Characters/Player/Male/Medium/Bodies/M_MED_FearlessFlightMenace/Meshes/Parts/M_MED_FearlessFlightMenace_FaceAcc_AnimBP
target_skeleton=&/M_MED_FearlessFlightMenace_FaceAcc_Skeleton
graphs [
 AnimGraph {nodes [
   n1 = Get Body_Mesh()
   n2 = Control Rig!(
    Source=Layered blend per bone(BasePose=Copy Pose From Mesh() {bCopyCurves=true, bUseAttachedParent=true}, BlendPoses_0=Copy Pose From Mesh(SourceMeshComponent=@n1)) {layerSetup [{branchFilters [{boneName=upperarm_l}, {boneName=upperarm_r}]}]}
    bAlphaBoolEnabled:bool=True
   ) -> (Pose) {
    alphaInputType=Bool
    alphaBoolBlend {blendOption=Linear}
    controlRigAssetReference {blueprintRigClass=&/M_MED_FearlessFlightMenace_FaceAcc_CtrlRig.M_MED_FearlessFlightMenace_FaceAcc_CtrlRig_C}
   }
   n3 = Control Rig(Source=@n2, bAlphaBoolEnabled=Get isRifle() -> (isRifle:bool)) {
    alphaInputType=Bool
    alphaBoolBlend {blendOption=Linear}
    controlRigAssetReference {blueprintRigClass=&/M_MED_FearlessFlightMenaceRifle_FaceAcc_CtrlRig.M_MED_FearlessFlightMenaceRifle_FaceAcc_CtrlRig_C}
   }
   n4 = Control Rig(Source=@n3, bAlphaBoolEnabled=Get isPistol() -> (isPistol:bool)) {
    alphaInputType=Bool
    alphaBoolBlend {blendOption=Linear}
    controlRigAssetReference {blueprintRigClass=&/M_MED_FearlessFlightMenacePistol_FaceAcc_CtrlRig.M_MED_FearlessFlightMenacePistol_FaceAcc_CtrlRig_C}
   }
   n5 = Get IsHangGlider() -> (IsHangGlider:double)
   n6 = "Transform (Modify) Bone - Bone: trap_l"(ComponentPose=Local To Component(LocalPose=@n4), Alpha=@n5, Translation="0,0,0", Rotation="0,0,0", Scale="0.8,1,1") {rotationMode=BMM_Additive, scaleMode=BMM_Additive, alphaBoolBlend {blendOption=Linear}, boneToModify {boneName=trap_l}, scale {x=0.8}}
   n7 = "Transform (Modify) Bone - Bone: trap_r"(ComponentPose=@n6, Alpha=@n5, Translation="0,0,0", Rotation="0,0,0", Scale="0.8,1,1") {rotationMode=BMM_Additive, scaleMode=BMM_Additive, alphaBoolBlend {blendOption=Linear}, boneToModify {boneName=trap_r}, scale {x=0.8}}
   n8 = "Transform (Modify) Bone - Bone: chest_l"(ComponentPose=@n7, Alpha=@n5, Translation="0,0,0", Rotation="0,0,0", Scale="0.6,1,1") {rotationMode=BMM_Additive, scaleMode=BMM_Additive, alphaBoolBlend {blendOption=Linear}, boneToModify {boneName=chest_l}, scale {x=0.6}}
   n9 = "Transform (Modify) Bone - Bone: chest_r"(ComponentPose=@n8, Alpha=@n5, Translation="0,0,0", Rotation="0,0,0", Scale="0.6,1,1") {rotationMode=BMM_Additive, scaleMode=BMM_Additive, alphaBoolBlend {blendOption=Linear}, boneToModify {boneName=chest_r}, scale {x=0.6}}
   n10 = Get isSkydiveFloatingFloat() -> (isSkydiveFloatingFloat:float)
   n11 = "Transform (Modify) Bone - Bone: trap_l"(ComponentPose=@n9, Alpha=@n10, Translation="0,0,3", Rotation="0,0,0", Scale="1,1,1") {
    rotationMode=BMM_Additive
    scaleMode=BMM_Additive
    translationMode=BMM_Additive
    alphaBoolBlend {blendOption=Linear}
    alphaScaleBiasClamp {bInterpResult=true, interpSpeedIncreasing=1}
    boneToModify {boneName=trap_l}
    translation {z=3}
   }
   n12 = "Transform (Modify) Bone - Bone: trap_r"(ComponentPose=@n11, Alpha=@n10, Translation="0,0,3", Rotation="0,0,0", Scale="1,1,1") {
    rotationMode=BMM_Additive
    scaleMode=BMM_Additive
    translationMode=BMM_Additive
    alphaBoolBlend {blendOption=Linear}
    alphaScaleBiasClamp {bInterpResult=true, interpSpeedIncreasing=1}
    boneToModify {boneName=trap_r}
    translation {z=3}
   }
   Output Pose(Result=Component To Local(ComponentPose=@n12)) {layerGroup=None}
   n13 = Get trapAlpha() -> (trapAlpha:double)
   n14 = Spring controller - Bone: upperarm_twist_01_r!(
    ComponentPose:Component Space Pose Link Structure=Spring controller - Bone: trap_r!(
     ComponentPose:Component Space Pose Link Structure=Spring controller - Bone: trap_l!(ComponentPose:Component Space Pose Link Structure, Alpha:float=@n13) -> (Pose:Component Space Pose Link Structure) {is=title AnimGraphNode_SpringBone, springDamping=7, springStiffness=500, alphaBoolBlend {blendOption=Linear}, springBone {boneName=trap_l}}
     Alpha:float=@n13
    ) -> (Pose:Component Space Pose Link Structure) {is=title AnimGraphNode_SpringBone, springDamping=7, springStiffness=500, alphaBoolBlend {blendOption=Linear}, springBone {boneName=trap_r}}
    Alpha:float=0.01
   ) -> (Pose:Component Space Pose Link Structure) {
    is=title AnimGraphNode_SpringBone
    springDamping=5
    springStiffness=300
    alphaBoolBlend {blendOption=Linear}
    springBone {boneName=upperarm_twist_01_r}
   }
   Spring controller - Bone: upperarm_twist_01_l!(ComponentPose:Component Space Pose Link Structure=@n14, Alpha:float=0.01) -> (Pose:Component Space Pose Link Structure) {
    is=title AnimGraphNode_SpringBone
    springDamping=5
    springStiffness=300
    alphaBoolBlend {blendOption=Linear}
    springBone {boneName=upperarm_twist_01_l}
   }
  ]}
 EventGraph {nodes [
   Event BlueprintUpdateAnimation()
   > n1 = Sequence!(execute) -> (then_0, then_1, then_2:Exec)
   > n2 = Cast To FortPlayerAnimInstance(Object=GetAnimInstance(self=GetOwnerSkeletalMeshComponent()))
   > Set isPistol(isPistol=false)
   > Set isRifle(isRifle:bool=false) -> (Output_Get:bool)
   > n3 = Switch on EFortWeaponCoreAnimation!(execute, Selection:EFortWeaponCoreAnimation Enum=Get WeaponCoreAnim(self=@n2)) -> (
    Melee:Exec
    Pistol:Exec
    Shotgun:Exec
    PaperBlueprint:Exec
    Rifle:Exec
    MeleeOneHand:Exec
    MachinePistol:Exec
    RocketLauncher:Exec
    GrenadeLauncher:Exec
    GoingCommando:Exec
    AssaultRifle:Exec
    TacticalShotgun:Exec
    SniperRifle:Exec
    TrapPlacement:Exec
    ShoulderLauncher:Exec
    AbilityDecoTool:Exec
    Crossbow:Exec
    C4:Exec
    RemoteControl:Exec
    DualWield:Exec
    AR_BullPup:Exec
    AR_ForwardGrip:Exec
    MedPackPaddles:Exec
    SMG_P90:Exec
    AR_DrumGun:Exec
    Consumable_Small:Exec
    Consumable_Large:Exec
    Balloon:Exec
    MountedTurret:Exec
    CreativeTool:Exec
    ExplosiveBow:Exec
    AshtonIndigo:Exec
    AshtonChicago:Exec
    MeleeDualWield:Exec
    Unarmed:Exec
   )
   Set isRifle(execute=[@n3.AssaultRifle, @n3.Rifle, @n3.RocketLauncher, @n3.Shotgun, @n3.SniperRifle, @n3.TacticalShotgun], isRifle=true)
   Set isPistol(execute=@n3.Pistol, isPistol=true)
   Set isSkydiveFloatingFloat(execute=@n1.then_1, isSkydiveFloatingFloat:float="To Float (Boolean)"(InBool=Get bIsSkydiveFloating())) -> (Output_Get:float)
   n4 = Branch(execute=@n1.then_2, Condition=Get isWearingBackbling())
   > Set trapAlpha(trapAlpha:double=0) -> (Output_Get:double)
   n5 = Branch(execute=@n4.else, Condition=OR Boolean(A=Get bIsJumping(), B=Get bIsFalling()))
   > Set trapAlpha(trapAlpha=0.005)
   Set trapAlpha(execute=@n5.else, trapAlpha=0.04)
   Event OnParachuteOpened()
   > Set Is Sky Diving(Is Sky Diving:double=0) -> (Output_Get:double)
   > n6 = Switch on EGliderType(Selection=Get PlayerGliderType() -> (PlayerGliderType=HangGlider))
   Set IsHangGlider(execute=[@n6.Cape, @n6.NoGlider, @n6.Surfing, @n6.SurfingSimple, @n6.Umbrella], IsHangGlider:double=0) -> (Output_Get:double)
   Delay(execute=@n6.HangGlider, Duration=1)
   > Set IsHangGlider(IsHangGlider=1)
   Event OnCharacterCustomizationLoaded()
   > Do Once()
   n7 = CheckForBackbling()
   > Set isWearingBackbling(isWearingBackbling=OR Boolean!(A=@n7, B=@n7.bIsBackblingACape, C:bool=@n7.bIsBackblingLong) -> (ReturnValue))
   Event BlueprintInitializeAnimation()
   > Set Body_Mesh(Body_Mesh=GetOwnerSkeletalMeshComponent())
   Event OnParachuteClosed()
   > Set IsHangGlider(IsHangGlider=0)
  ]}
]
variables [
 Body_Mesh {type=SkeletalMeshComponent}
 Is Sky Diving {type=number}
 IsHangGlider {type=number}
 isPistol {type=boolean}
 isRifle {type=boolean}
 isSkydiveFloatingFloat {type=number}
 isWearingBackbling {type=boolean}
 trapAlpha {type=number}
]
